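# Translucent UI material for Discovery Browser category tile backgrounds, built from a Thumbnail texture with gradient and rounded-corner outline masks.
# Samples the fill-scaled thumbnail, then eases it into a tint taken from two near-point thumbnail samples, adjusted in HSV.
# A 16-tap averaged blur provides another color output.
# Gradient masks blend the color layers, and IsThumbnailLoaded fades the result in over the Slate post buffer.
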Material /DiscoveryBrowser/Tiles/Categories/Materials/M_UI_Category_Background
graph {
 properties {
  bAutomaticallySetUsageInEditor=false
  bUsedWithStaticMesh=false
  blendMode=BLEND_Translucent
  materialDomain=MD_UI
  pixelDepthOffsetMode=PDOM_Legacy
  usedShadingModels=Default Lit
 }
 nodes [
  Constant2Vector() {r=0.5}
  Constant2Vector() {g=0.52, r=0.45}
  n1 = NamedRerouteUsage()
  n2 = Normalize(VectorInput=Max(A=@n1) {constB=0})
  n3 = Saturate(None=Divide(A=Saturate(None=@n2)) {constA=0.1, constB=0.6})
  n4 = MaterialFunctionCall() {materialFunction=&GetUserInterfaceUV}
  n5 = NamedRerouteDeclaration(None=OneMinus(None=ComponentMask(None=@n4.Normalized UV) {r=true})) {name=Base Gradient Mask, nodeColor=<rgba 0 0 0 1>}
  n6 = LinearInterpolate(Alpha=Multiply(A=@n5, B=NamedRerouteUsage())) {constB=0.1}
  n7 = ScalarParameter() {defaultValue=1, parameterName=Gradient Opacity, sliderMax=1}
  n8 = MaterialFunctionCall(
   Input=Saturate(None=@n5)
   Input Low=NamedRerouteDeclaration(None=Constant() {r=0.2}) {name=Gradient Range, nodeColor=<rgba 0 0 0 1>}
   Input High=Constant() {r=1}
  ) {materialFunction=&RemapValue_0-1}
  n9 = NamedRerouteDeclaration(None=SmoothStep(Value=Saturate(None=@n8.Result))) {name=Gradient Mask, nodeColor=<rgba 0 0 0 1>}
  LinearInterpolate(
   A=MaterialFunctionCall!("UVs (Optional)") {materialFunction=/Engine/Functions/UserInterface/GetSlatePost0}.LinearRGB
   B=LinearInterpolate(A=NamedRerouteUsage(), B=LinearInterpolate(A=@n1, B=@n3, Alpha=@n6), Alpha=LinearInterpolate(B=@n7, Alpha=@n9))
   Alpha=ScalarParameter() {defaultValue=1, parameterName=IsThumbnailLoaded, sliderMax=1}
  )
  n10 = NamedRerouteDeclaration(None=TextureObjectParameter() {parameterName=Thumbnail}) {name=Thumbnail Object, nodeColor=<rgba 0 0 0 1>}
  n11 = MaterialFunctionCall(
   UVs=MaterialFunctionCall(UV=@n4.Normalized UV, TextureSize=TextureProperty(Tex=@n10)) {materialFunction=&MF_UI_UVScaling}.Scaled To Fill
   Texture Scale=ScalarParameter() {defaultValue=1, parameterName=Scale}
  ) {materialFunction=&ScaleUVsByCenter}
  n12 = NamedRerouteUsage()
  n13 = TextureSample(UVs=@n11.UVs, Tex=@n12) {samplerSource=SSM_Clamp_WorldGroupSettings}
  n14 = LinearInterpolate(Alpha=NamedRerouteUsage()) {constA=1, constB=0.1}
  n15 = NamedRerouteUsage()
  n16 = Add(A=@n11.UVs, B=AppendVector(A=OneMinus(None=@n15), B=Constant()))
  n17 = AppendVector(
   A=ScalarParameter() {defaultValue=0.428, group=Color Adjustments, parameterName=ColorPos01, sliderMax=1}
   B=ScalarParameter() {defaultValue=0.8, group=Color Adjustments, parameterName=ColorPos02, sliderMax=1}
  )
  n18 = Subtract(A=@n16, B=@n17)
  n19 = TextureSample(UVs=Add(A=Multiply(A=@n18) {constB=0.0017}, B=@n17), Tex=@n12) {samplerSource=SSM_Clamp_WorldGroupSettings}
  n20 = AppendVector(
   A=ScalarParameter() {defaultValue=0.52, group=Color Adjustments, parameterName=ColorPos03, sliderMax=1}
   B=ScalarParameter() {defaultValue=0.55, group=Color Adjustments, parameterName=ColorPos04, sliderMax=1}
  )
  n21 = Subtract(A=@n16, B=@n20)
  n22 = TextureSample(UVs=Add(A=Multiply(A=@n21) {constB=0.0017}, B=@n20), Tex=@n12) {samplerSource=SSM_Clamp_WorldGroupSettings}
  n23 = MaterialFunctionCall() {materialFunction=&GetUserInterfaceUV}
  n24 = MaterialFunctionCall(UVs=Add(A=OneMinus(None=@n23.Normalized UV), B=Constant2Vector() {r=-0.4}), "Rotation Angle (0-1)"=Constant() {r=-0.075}) {materialFunction=&CustomRotator}
  n25 = SmoothStep(Value=Saturate(None=ComponentMask(None=@n24.Rotated Values) {r=true}))
  n26 = MaterialFunctionCall(Float3=RgbToHsv!(None=LinearInterpolate(A=@n19.RGB, B=@n22.RGB, Alpha=@n25))) {materialFunction=&BreakOutFloat3Components}
  n27 = ScalarParameter() {defaultValue=0.2, group=Color Adjustments, parameterName=ColorBrightness, sliderMax=1}
  n28 = MaterialFunctionCall(
   R=@n26.R
   G=LinearInterpolate(A=@n26.G, Alpha=ScalarParameter() {defaultValue=0.1, group=Color Adjustments, parameterName=ColorSaturation, sliderMax=1}) {constAlpha=0.1}
   B=Min(A=LinearInterpolate(A=@n26.B, Alpha=@n27) {constAlpha=0.2}) {constB=0.75}
  ) {materialFunction=&AppendMany}
  n29 = Saturate(None=MaterialFunctionCall!(Base=HsvToRgb!(None=@n28.RGB), Add=ScalarParameter() {defaultValue=0.1, group=Color Adjustments, parameterName=ColorIntensity, sliderMax=1}) {materialFunction=/Engine/Functions/Engine_MaterialFunctions02/Math/MultiplyAdd}.Result)
  n30 = Multiply(A=@n29, B=Power(Base=@n29, Exp=ScalarParameter() {defaultValue=1, group=Color Adjustments, parameterName=ColorContrast}))
  n31 = ScalarParameter() {defaultValue=1, group=Color Adjustments, parameterName=ColorLuminosity, sliderMax=1}
  n32 = NamedRerouteUsage()
  NamedRerouteDeclaration(None=LinearInterpolate(
    A=Multiply(A=@n13.RGB, B=@n14)
    B=Multiply(A=@n30, B=@n31)
    Alpha=MaterialFunctionCall("Value (0-1)"=@n32) {materialFunction=&MF_UI_EaseCurves}.EaseOutCubic
   )) {name=Texture RGB, nodeColor=<rgba 0 0 0 1>}
  n33 = MaterialFunctionCall(
   Corner Radius=ScalarParameter() {defaultValue=20, parameterName=Corner Radius}
   Radius in Pixels=Constant() {r=1}
   Outline Thickness=Constant() {r=4}
  ) {materialFunction=&MF_UI_CoreUI_RoundedMask}
  n34 = OneMinus(None=@n33.Shape Mask)
  NamedRerouteDeclaration(None=Max(A=Ceil(None=@n34), B=@n33.Outline Mask)) {name=Outline Mask, nodeColor=<rgba 0 0 0 1>}
  n35 = NamedRerouteUsage()
  n36 = Multiply(A=@n35, B=Constant2Vector() {g=-1, r=1})
  n37 = NamedRerouteUsage()
  n38 = NamedRerouteUsage()
  n39 = Multiply(A=@n38, B=Constant2Vector() {g=-1, r=-1})
  n40 = Add(
   A=TextureSample(UVs=Add(A=@n38, B=@n39), Tex=@n37) {samplerSource=SSM_Wrap_WorldGroupSettings}.RGB
   B=TextureSample(UVs=Add(A=@n35, B=@n36), Tex=@n37) {samplerSource=SSM_Wrap_WorldGroupSettings}.RGB
  )
  n41 = NamedRerouteUsage()
  n42 = Constant2Vector() {g=1, r=-1}
  n43 = TextureSample(UVs=Add(A=@n41, B=Multiply(A=@n41, B=@n42)), Tex=@n37) {samplerSource=SSM_Wrap_WorldGroupSettings}
  n44 = NamedRerouteUsage()
  n45 = Add(A=@n44, B=Multiply(A=@n44, B=Constant2Vector() {g=1, r=1}))
  n46 = NamedRerouteUsage()
  n47 = Multiply(A=@n46, B=Constant2Vector() {g=-0.5, r=0.5})
  n48 = NamedRerouteUsage()
  n49 = Multiply(A=@n48, B=Constant2Vector() {g=-0.5, r=-0.5})
  n50 = Add(
   A=TextureSample(UVs=Add(A=@n48, B=@n49), Tex=@n37) {samplerSource=SSM_Wrap_WorldGroupSettings}.RGB
   B=TextureSample(UVs=Add(A=@n46, B=@n47), Tex=@n37) {samplerSource=SSM_Wrap_WorldGroupSettings}.RGB
  )
  n51 = NamedRerouteUsage()
  n52 = Constant2Vector() {g=0.5, r=-0.5}
  n53 = TextureSample(UVs=Add(A=@n51, B=Multiply(A=@n51, B=@n52)), Tex=@n37) {samplerSource=SSM_Wrap_WorldGroupSettings}
  n54 = NamedRerouteUsage()
  n55 = Add(A=@n54, B=Multiply(A=@n54, B=Constant2Vector() {g=0.5, r=0.5}))
  n56 = Add(
   A=Add(A=Add(A=@n40, B=@n43.RGB), B=TextureSample(UVs=@n45, Tex=@n37) {samplerSource=SSM_Wrap_WorldGroupSettings}.RGB)
   B=Add(A=Add(A=@n50, B=@n53.RGB), B=TextureSample(UVs=@n55, Tex=@n37) {samplerSource=SSM_Wrap_WorldGroupSettings}.RGB)
  )
  n57 = NamedRerouteUsage()
  n58 = Multiply(A=@n57, B=Constant2Vector() {r=0.5})
  n59 = NamedRerouteUsage()
  n60 = NamedRerouteUsage()
  n61 = Multiply(A=@n60, B=Constant2Vector() {r=-0.5})
  n62 = Add(
   A=TextureSample(UVs=Add(A=@n60, B=@n61), Tex=@n59) {samplerSource=SSM_Wrap_WorldGroupSettings}.RGB
   B=TextureSample(UVs=Add(A=@n57, B=@n58), Tex=@n59) {samplerSource=SSM_Wrap_WorldGroupSettings}.RGB
  )
  n63 = NamedRerouteUsage()
  n64 = Constant2Vector() {g=0.5}
  n65 = TextureSample(UVs=Add(A=@n63, B=Multiply(A=@n63, B=@n64)), Tex=@n59) {samplerSource=SSM_Wrap_WorldGroupSettings}
  n66 = NamedRerouteUsage()
  n67 = Add(A=@n66, B=Multiply(A=@n66, B=Constant2Vector() {g=0.5}))
  n68 = NamedRerouteUsage()
  n69 = Multiply(A=@n68, B=Constant2Vector() {r=1})
  n70 = NamedRerouteUsage()
  n71 = Multiply(A=@n70, B=Constant2Vector() {r=-1})
  n72 = Add(
   A=TextureSample(UVs=Add(A=@n70, B=@n71), Tex=@n59) {samplerSource=SSM_Wrap_WorldGroupSettings}.RGB
   B=TextureSample(UVs=Add(A=@n68, B=@n69), Tex=@n59) {samplerSource=SSM_Wrap_WorldGroupSettings}.RGB
  )
  n73 = NamedRerouteUsage()
  n74 = Constant2Vector() {g=1}
  n75 = TextureSample(UVs=Add(A=@n73, B=Multiply(A=@n73, B=@n74)), Tex=@n59) {samplerSource=SSM_Wrap_WorldGroupSettings}
  n76 = NamedRerouteUsage()
  n77 = Add(A=@n76, B=Multiply(A=@n76, B=Constant2Vector() {g=1}))
  n78 = Add(
   A=Add(A=Add(A=@n62, B=@n65.RGB), B=TextureSample(UVs=@n67, Tex=@n59) {samplerSource=SSM_Wrap_WorldGroupSettings}.RGB)
   B=Add(A=Add(A=@n72, B=@n75.RGB), B=TextureSample(UVs=@n77, Tex=@n59) {samplerSource=SSM_Wrap_WorldGroupSettings}.RGB)
  )
  NamedRerouteDeclaration(None=Multiply(A=Add(A=@n56, B=@n78)) {constB=0.0625}) {name=BG_Blured_RGB_Output, nodeColor=<rgba 0 0 0 1>}
  NamedRerouteUsage()
 ]
}
