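# Category tile button for the Discovery browser, showing a title and taking its tile view model from the list item.
# Clicking fires an interaction event, notifies the activity selector, then asynchronously pushes a page.
# With no reference panel name it pushes the nested surface, passing view model and surface name; otherwise the View All page for that surface and panel.
# Hover takes focus, and the outline follows focus except on touch, clearing when input switches to touch.
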WidgetBlueprint /DiscoveryBrowser/Tiles/Categories/WBP_CategoryTile
graphs [
 EventGraph {nodes [
   n1 = HandleInputMethodChanged!() -> (OutputDelegate, then, bNewInputType:ECommonInputType Enum) {is=title K2Node_CustomEvent}
   > Branch(Condition="Equal (Enum)"(A=@n1.bNewInputType, B=Touch))
   > RequestTransition!(
    execute
    self:UIKit Block Object Reference=Get Block_Outline() -> (Block_Outline:WBP UIKit Block Outline Object Reference)
    State:EUIKitBlockVisualState Enum=Unfocused
   ) -> (then)
   Event On Clicked()
   > Fire Event Element Interacted(Element=Self-Reference())
   > HandleCategoryTileClicked!(
    execute
    self:Fort Activity Selector Object Reference=Get FortActivitySelector() -> (ReturnValue:Fort Activity Selector Object Reference) {is=title K2Node_GetSubsystem}
    FortPC:Fort Player Controller Object Reference=GetFirstLocalFortPlayerController!(bIncludeNonSpectators:bool=true, bIncludeSpectators:bool=false) -> (ReturnValue:Fort Player Controller Object Reference)
    InTileItemVM:Fort Discover Surface Tile Item VM Object Reference=Get FortDiscoverSurfaceTileItemVM() -> (FortDiscoverSurfaceTileItemVM:Fort Discover Surface Tile Item VM Object Reference)
   ) -> (then)
   n2 = Get FortDiscoverSurfaceTileItemVM()
   > n3 = Branch(Condition="Equal (Name)"(A=GetReferencePanelName!(self:Fort Discover Surface Tile Item VM Object Reference=@n2) -> (ReturnValue:Name=None)))
   > n4 = PushPrimaryContentClassAsync!(
    execute:Exec
    WidgetClass:Common Activatable Widget Class Reference=/DiscoveryBrowser/WBP_Discover_NestedSurface.WBP_Discover_NestedSurface_C
   ) -> (then:Exec, OnPrimaryContentPushed:Exec, ActivatableWidget:Common Activatable Widget Object Reference) {is=title K2Node_AsyncAction}
   n5 = Cast To WBP_Discover_NestedSurface!(execute=@n4.OnPrimaryContentPushed, Object=@n4) -> (then, CastFailed, AsWBP Discover Nested Surface:WBP Discover Nested Surface Object Reference) {is=title K2Node_DynamicCast}
   n6 = Get FortDiscoverSurfaceTileItemVM()
   > SetFortDiscoverSurfaceTileItemVM!(execute, self:WBP Discover Nested Surface Object Reference=@n5, Viewmodel:Fort Discover Surface Tile Item VM Object Reference=@n6) -> (then)
   > On Disover Surface Tile Item VM Set(self:WBP Discover Nested Surface Object Reference=@n5)
   n7 = GetReferenceSurfaceName!(self:Fort Discover Surface Tile Item VM Object Reference=@n6) -> (ReturnValue:Name=None)
   > SetSurfaceName!(
    execute
    self:Discover Surface View Model Object Reference=Get FortDiscoverSurfaceViewModel!(self:WBP Discover Nested Surface Object Reference=@n5) -> (FortDiscoverSurfaceViewModel:Discover Surface View Model Object Reference)
    InSurfaceName:Name=@n7
   ) -> (then)
   n8 = PushPrimaryContentClassAsync(execute=@n3.else, WidgetClass=/DiscoveryBrowser/WBP_ViewAllPage.WBP_ViewAllPage_C)
   n9 = Cast To WBP_ViewAllPage!(execute=@n8.OnPrimaryContentPushed, Object=@n8) -> (then, CastFailed, AsWBP View All Page:WBP View All Page Object Reference) {is=title K2Node_DynamicCast}
   > SetupSurface!(
    execute
    self:WBP View All Page Object Reference=@n9
    Surface Name:Name=@n7
    Panel Name:Name=GetReferencePanelName(self=@n6) -> (ReturnValue=None)
   ) -> (then)
   > ActivateWidget(self=@n8)
   Event OnInitialized()
   > Bind Event to On Input Method Changed(self=Get CommonInputSubsystem(), Delegate=Create Event())
   Event On Hovered()
   > SetFocus()
   n10 = Event OnListItemObjectSet()
   > n11 = Cast To FortDiscoverSurfaceTileItemVM!(execute, Object=@n10.ListItemObject) -> (then, CastFailed, AsFort Discover Surface Tile Item VM:Fort Discover Surface Tile Item VM Object Reference) {is=title K2Node_DynamicCast}
   > Set FortDiscoverSurfaceTileItemVM(FortDiscoverSurfaceTileItemVM:Fort Discover Surface Tile Item VM Object Reference=@n11) -> (Output_Get:Fort Discover Surface Tile Item VM Object Reference)
   Event On Focused() {is=title K2Node_Event}
   n12 = IsUsingTouch(PlayerController=GetOwningPlayer())
   > n13 = Branch(Condition=@n12)
   n14 = Get Block_Outline()
   RequestTransition(execute=@n13.else, self=@n14, State=Focused)
   Event On Unfocused() {is=title K2Node_Event}
   > n15 = Branch(Condition=@n12)
   RequestTransition(execute=@n15.else, self=@n14, State=Unfocused)
  ]}
 Update Display Name {nodes [
   n1 = Update Display Name!() -> (then, In Display Name:Text) {is=title K2Node_FunctionEntry}
   > Set Text(Text=@n1, self=Get Text_CategoryTitle() -> (Text_CategoryTitle:Common Text Object Reference))
  ]}
]
variables [ActivityCategoryPageViewClass {type=Class, default=&WBP_Discover_NestedSurface_C}]
widgets {
 info {inheritedWidgetCount=1, parentClass=&CommonButtonBase, rootWidgetClass=&GridPanel, widgetCount=6}
 widgets [
  GridPanel_0 {widgetClassPath=&GridPanel, widget=GridPanel_0, properties {columnFill [1], rowFill [1]}}
  UIKit_Backplate {
   widgetClassPath=&WBP_UIKit_Backplate
   parent=GridPanel_0
   widget=UIKit_Backplate
   properties {brightness=Medium}
   slot {class=GridSlot}
  }
  SizeBox {
   widgetClassPath=&SizeBox
   parent=GridPanel_0
   widget=SizeBox
   properties {clipping=ClipToBounds}
   slot {class=GridSlot, properties {layer=1, padding=<ltrb 16 16 16 16>}}
  }
  Text_CategoryTitle {
   widgetClassPath=&CommonTextBlock
   bIsVariable=true
   parent=SizeBox
   widget=Text_CategoryTitle
   properties {
    applyLineHeightToBottomLine=false
    autoWrapText=true
    justification=Center
    text="{Category}\r\n{Category}"
    textOverflowPolicy=Ellipsis
    font {
     bForceMonospaced=false
     bMaterialIsStencil=false
     fontMaterial=/FortUILibrary/MaterialInstances/Text/MI_UI_AdaptiveText_Light
     fontObject=&HeadingNow
     letterSpacing=10
     monospacedWidth=1
     size=18
     skewAmount=0
     typefaceFontName=75 Medium
     outlineSettings {
      bApplyOutlineToDropShadows=false
      bMiteredCorners=false
      bSeparateFillAlpha=false
      outlineMaterial=null
      outlineSize=0
      outlineColor=<rgba 0 0 0 1>
     }
    }
    shadowOffset=<0 0>
   }
   slot {class=SizeBoxSlot, properties {verticalAlignment=VAlign_Center}}
  }
  Block_Outline {
   widgetClassPath=/FortUILibrary/UIKit/Buttons/Blocks/WBP_UIKit_Block_Outline
   bIsVariable=true
   parent=GridPanel_0
   widget=Block_Outline
   properties {bottomLeftCornerRadius=20, bottomRightCornerRadius=20, radiusInPixels=true, topLeftCornerRadius=20, topRightCornerRadius=20}
   slot {class=GridSlot, properties {layer=2}}
  }
  InputActionWidget {widgetClassPath=&CommonActionWidget, bInherited=true, widget=null}
 ]
}
