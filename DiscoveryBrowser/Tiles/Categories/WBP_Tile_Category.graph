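# Discovery browser category tile button that shows a category title and thumbnail, driven by a surface tile item view model.
# Clicking reports telemetry and pushes a nested surface, or toggles pending favorite with sound in the interests chooser row.
# Context menu actions offer View Category and Favorite/Unfavorite, with a favorite status indicator.
# Widget animations feed background material parameters for hover scale and thumbnail load, and mobile gets a larger tile.
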WidgetBlueprint /DiscoveryBrowser/Tiles/Categories/WBP_Tile_Category
animations [
 IsHoveredIncrease {
  binding_count=1
  length_seconds=0.2017
  object=&self:IsHoveredIncrease
  sequence {
   bindings [WBP_Tile_Category_C_40 {
     id=#1
     possessed_class=WBP_Tile_Category_C
     tracks [{
       type=MovieSceneDoubleTrack
       display_name=Anim Value
       sections [{type=MovieSceneDoubleSection, channels [None {type=MovieSceneScriptingDoubleChannel, keys [[0,1],[12.103,1.1]]}]}]
      }]
    }]
   display_rate {numerator=60}
  }
 }
 Anim_ThumbnailLoaded {
  binding_count=1
  length_seconds=0.6
  object=&self:Anim_ThumbnailLoaded
  sequence {
   bindings [WBP_Tile_Category_C_21 {
     id=#2
     possessed_class=WBP_Tile_Category_C
     tracks [{
       type=MovieSceneDoubleTrack
       display_name=Loaded Value
       sections [{type=MovieSceneDoubleSection, channels [None {type=MovieSceneScriptingDoubleChannel, keys [[0,0],[36,1]]}]}]
      }]
    }]
   display_rate {numerator=60}
  }
 }
]
graphs [
 AddViewCategoryToContextMenu {nodes [
   n1 = AddViewCategoryToContextMenu!() -> (then, IsInInterestsRow:bool=~) {is=title K2Node_FunctionEntry}
   > n2 = Branch(Condition=@n1)
   SetupViewCategory(execute=@n2.else)
  ]}
 EventGraph {nodes [
   ToggleFavorite!() -> (OutputDelegate, then, InteractionNameID:Name) {is=title K2Node_CustomEvent}
   > n1 = Branch(Condition=Get bIsInInterestsChooserRow!(self:Discover Tile Item View Model Object Reference=Get FortDiscoverSurfaceTileItemVM() -> (FortDiscoverSurfaceTileItemVM:Fort Discover Surface Tile Item VM Object Reference)) -> (bIsInInterestsChooserRow:bool))
   n2 = Get FortDiscoverSurfaceTileItemVM()
   > TogglePendingFavorite(self:Fort Discover Surface Tile Item VM Object Reference=@n2)
   > n3 = Branch(Condition=Get bIsPendingFavorite!(self:Fort Discover Surface Tile Item VM Object Reference=@n2) -> (bIsPendingFavorite:bool))
   n4 = PlaySound2D(execute=[@n3.then, @n6.then], Sound=Get Favorited Sound() -> (Favorited Sound:Sound Base Object Reference))
   CloseContextMenu(
    execute=[@n5.then, @n4.then]
    self:WBP Discover Tile Base Object Reference=Get WBP_Discover_Tile_Base() -> (WBP_Discover_Tile_Base:WBP Discover Tile Base Object Reference)
   )
   n5 = PlaySound2D(execute=[@n3.else, @n6.else], Sound=Get Unfavorited Sound() -> (Unfavorited Sound:Sound Base Object Reference))
   ToggleFavorite(execute=@n1.else, self:Fort Discover Link Entry Tile Item VM Object Reference=@n2) {is=title K2Node_CallFunction}
   > n6 = Branch(Condition=Get bIsFavorite!(self:Fort Discover Link Entry Tile Item VM Object Reference=@n2) -> (bIsFavorite:bool))
   ViewCategory!() -> (OutputDelegate, then, InteractionID:Name) {is=title K2Node_CustomEvent}
   > n7 = CloseContextMenu(self=Get WBP_Discover_Tile_Base())
   n8 = Get FortDiscoverSurfaceTileItemVM()
   n9 = GetReferenceSurfaceName!(self:Fort Discover Surface Tile Item VM Object Reference=@n8) -> (ReturnValue:Name=None)
   SendFlowRiderClickTileEvent!(
    execute=[@n7.then, @n23.then]
    self:WBP Discover Tile Base Object Reference=Get WBP_Discover_Tile_Base()
    TileWidget:User Widget Object Reference=Self-Reference()
    TileType:EActivityItemType Enum=SurfaceTile
    TileUniqueIdentifier:String="To String (Name)"(InName=@n9)
   ) -> (then)
   > HandleCategoryTileClicked!(
    execute
    self:Fort Activity Selector Object Reference=Get FortActivitySelector() -> (ReturnValue:Fort Activity Selector Object Reference) {is=title K2Node_GetSubsystem}
    FortPC:Fort Player Controller Object Reference=GetFirstLocalFortPlayerController!(bIncludeNonSpectators:bool=true, bIncludeSpectators:bool=false) -> (ReturnValue:Fort Player Controller Object Reference)
    InTileItemVM:Fort Discover Surface Tile Item VM Object Reference=Get FortDiscoverSurfaceTileItemVM()
   ) -> (then)
   > n10 = PushPrimaryContentClassAsync!(
    execute:Exec
    WidgetClass:Common Activatable Widget Class Reference=/DiscoveryBrowser/WBP_Discover_NestedSurface.WBP_Discover_NestedSurface_C
   ) -> (then:Exec, OnPrimaryContentPushed:Exec, ActivatableWidget:Common Activatable Widget Object Reference) {is=title K2Node_AsyncAction}
   n11 = Cast To WBP_Discover_NestedSurface!(execute=@n10.OnPrimaryContentPushed, Object=@n10) -> (then, CastFailed, AsWBP Discover Nested Surface:WBP Discover Nested Surface Object Reference) {is=title K2Node_DynamicCast}
   > SetFortDiscoverSurfaceTileItemVM!(execute, self:WBP Discover Nested Surface Object Reference=@n11, Viewmodel:Fort Discover Surface Tile Item VM Object Reference=@n8) -> (then)
   > On Disover Surface Tile Item VM Set(self:WBP Discover Nested Surface Object Reference=@n11)
   > SetSurfaceName!(
    execute
    self:Discover Surface View Model Object Reference=Get FortDiscoverSurfaceViewModel!(self:WBP Discover Nested Surface Object Reference=@n11) -> (FortDiscoverSurfaceViewModel:Discover Surface View Model Object Reference)
    InSurfaceName:Name=@n9
   ) -> (then)
   SetupViewCategory() {is=title K2Node_CustomEvent}
   n12 = Get WBP_Discover_Tile_Base()
   > RemoveInteractionByName!(
    execute
    self:WBP Discover Tile Base Object Reference=@n12
    ActionToRemove:Name=Get Action to Remove_Category() -> (Action to Remove_Category:Name=None)
   ) -> (then)
   > AddInteraction!(
    execute
    self:UIKit Context Menu Action Context Object Reference=Get ContextMenuActionContext!(self:WBP Discover Tile Base Object Reference=@n12) -> (ContextMenuActionContext:UIKit Context Menu Action Context Object Reference)
    InActionInfo_DisplayName:Text="LOCTEXT(\"View Category\")"
    InActionInfo_IconTexture:Texture 2D Soft Object Reference
    InActionInfo_IconMaterial:Material Instance Soft Object Reference
    InActionInfo_SortPriority:int=1000
    InActionInfo_bIsAvailable:bool=True
    InActionInfo_bIsEnabled:bool=True
    InActionInfo_bIsSelected:bool=False
    InActionInfo_bAddDivider:bool=False
    InActionInfo_bPositionDividerAfter:bool=False
    InActionInfo_bCloseMenuOnSelect:bool=False
    InActionInfo_InteractionNameID:Name=Get Action to Remove_Category() -> (Action to Remove_Category=None)
    InActionInfo_Interaction:Data Table Row Handle Structure=(DataTable=None,RowName="")
    InActionInfo_PressedDelegate:Delegate="(null).None"
    InActionInfo_ReleasedDelegate:Delegate="(null).None"
    InInteractionDelegate:"Delegate (by ref)"=Create Event()
   ) -> (then)
   n13 = Event PreConstruct()
   > Parent: PreConstruct(IsDesignTime=@n13.IsDesignTime)
   > SetBackgroundMaterial!(
    execute
    self:WBP UIKit Block Background Base Object Reference=Get Block_Background() -> (Block_Background:WBP UIKit Block Background Generic Object Reference)
    Material Instance:"Material Instance Object Reference (by ref)"=Get BackgroundMaterialInstance() -> (BackgroundMaterialInstance:Material Instance Object Reference)
   ) -> (then)
   > Set Tile Size!(execute, self:WBP Discover Tile Base Object Reference=Get WBP_Discover_Tile_Base(), InSize:Vector 2D Structure=(X=396,Y=144)) -> (then) {is=title K2Node_CallFunction}
   > SetCompactUIStyle() {is=title K2Node_CallFunction}
   n14 = Event OnListItemObjectSet()
   > n15 = Cast To FortDiscoverSurfaceTileItemVM!(execute, Object=@n14.ListItemObject) -> (then, CastFailed, AsFort Discover Surface Tile Item VM:Fort Discover Surface Tile Item VM Object Reference) {is=title K2Node_DynamicCast}
   > n16 = Set FortDiscoverSurfaceTileItemVM(FortDiscoverSurfaceTileItemVM:Fort Discover Surface Tile Item VM Object Reference=@n15) -> (Output_Get:Fort Discover Surface Tile Item VM Object Reference)
   > n17 = Branch(Condition=HasFetchedImages!(self:Fort Discover Link Entry Tile Item VM Object Reference=@n16) -> (ReturnValue:bool))
   ConfigureDebugUI!(
    execute=[@n18.then, @n17.then]
    self:WBP Discover Tile Base Object Reference=Get WBP_Discover_Tile_Base()
    Debug VM:Discover Tile Item Debug View Model Object Reference=Get DebugVM!(self:Discover Tile Item View Model Object Reference=Get FortDiscoverSurfaceTileItemVM()) -> (DebugVM:Discover Tile Item Debug View Model Object Reference)
    OwningButton:Common Button Base Object Reference=Self-Reference()
   ) -> (then)
   > SetTileItemViewModel!(
    execute
    self:WBP Discover Tile Base Object Reference=Get WBP_Discover_Tile_Base()
    TileItemViewModel:Discover Tile Item View Model Object Reference=@n16
   ) -> (then)
   n18 = FetchImages(execute=@n17.else, self:Fort Discover Link Entry Tile Item VM Object Reference=@n16)
   Event Construct()
   > Parent: Construct()
   Event On Clicked()
   > n19 = Branch(Condition=Get bIsInInterestsChooserRow(self=Get FortDiscoverSurfaceTileItemVM()))
   > Branch(Condition=Get bEnableFortniteInterests!(self:Fort Runtime Options Object Reference=GetRuntimeOptions()) -> (bEnableFortniteInterests:bool))
   > TogglePendingFavorite(self=Get FortDiscoverSurfaceTileItemVM())
   > n20 = Branch(Condition=Get bIsPendingFavorite(self=Get FortDiscoverSurfaceTileItemVM()))
   > n21 = PlaySound2D(Sound=Get Favorited Sound())
   HandleCategoryTileClicked(
    execute=[@n21.then, @n22.then]
    self=Get FortActivitySelector()
    FortPC=GetFirstLocalFortPlayerController(bIncludeNonSpectators=true, bIncludeSpectators=false)
    InTileItemVM=Get FortDiscoverSurfaceTileItemVM()
   )
   n22 = PlaySound2D(execute=@n20.else, Sound=Get Unfavorited Sound())
   n23 = Parent: On Clicked(execute=@n19.else) {is=title K2Node_CallParentFunction}
   Event On Unhovered()
   > Parent: On Unhovered() {is=title K2Node_CallParentFunction}
   n24 = Get IsHoveredIncrease() -> (IsHoveredIncrease:Widget Animation Object Reference)
   > PlayAnimationReverse(InAnimation=@n24)
   n25 = Event SetNestedButtonGroupSelectionVM!() -> (OutputDelegate, then, InVM:Fort Nested Button Group Selection VM Object Reference) {is=title K2Node_Event}
   > Set FortNestedButtonGroupSelectionVM(FortNestedButtonGroupSelectionVM:Fort Nested Button Group Selection VM Object Reference=@n25.InVM) -> (Output_Get:Fort Nested Button Group Selection VM Object Reference)
   Event On Hovered()
   > Parent: On Hovered() {is=title K2Node_CallParentFunction}
   > PlayAnimationForward(InAnimation=@n24)
  ]}
 GetBackgroundTexture {nodes [
   n1 = GetBackgroundTexture!() -> (then, ImageLoadingState:EActivityImageLoadingState Enum=Unset, ImageTexture:Texture Object Reference) {is=title K2Node_FunctionEntry}
   > n2 = Switch on EActivityImageLoadingState!(execute, Selection:EActivityImageLoadingState Enum=@n1) -> (Unset:Exec, Loading:Exec, Loaded:Exec, Error:Exec) {is=title K2Node_SwitchEnum}
   SetScalarParameterValue(
    execute=[@n2.Error, @n2.Loading, @n2.Unset]
    self=Get BackgroundDynamicMaterial!(self:WBP UIKit Block Background Base Object Reference=Get Block_Background()) -> (BackgroundDynamicMaterial:Material Instance Dynamic Object Reference)
    ParameterName=IsThumbnailLoaded
    Value=0
   )
   > Return Node!(execute, Texture:Texture Object Reference=/Game/UI/DiscoveryShared/Art/T_UI_Discovery_DefaultEmpty)
   Return Node(execute=@n2.Loaded, Texture=@n1.ImageTexture)
  ]}
 GetFavoriteStatus {nodes [
   GetFavoriteStatus!() -> (then, IsPendingFavorite:bool=~, IsFavorite:bool=~, IsInInterestsRow:bool=~) {is=title K2Node_FunctionEntry}
   > n1 = Branch(Condition=Get bEnableFortniteInterests(self=GetRuntimeOptions()))
   > n2 = Branch(Condition=Get bIsInInterestsChooserRow(self=Get FortDiscoverSurfaceTileItemVM()))
   n3 = Select(
    Option 0:bool=Get IsFavorite() -> (IsFavorite:bool)
    Option 1:bool=Get IsPendingFavorite() -> (IsPendingFavorite:bool)
    Index=Get IsInInterestsRow() -> (IsInInterestsRow:bool)
   ) -> (ReturnValue:bool)
   Return Node!(
    execute=[@n5.then, @n2.then]
    Visibility:ESlateVisibility Enum=Select(Option 0:ESlateVisibility Enum=Collapsed, Option 1:ESlateVisibility Enum=SelfHitTestInvisible, Index=@n3) -> (ReturnValue:ESlateVisibility Enum)
   )
   n4 = Get WBP_Discover_Tile_Base()
   RemoveInteractionByName(execute=@n2.else, self=@n4, ActionToRemove=Get Action to Remove_Favorite() -> (Action to Remove_Favorite:Name=None))
   > n5 = AddInteraction(
    self=Get ContextMenuActionContext(self=@n4)
    InActionInfo_DisplayName=Select(Option 0:Text="LOCTEXT(\"Favorite\")", Option 1:Text="LOCTEXT(\"Unfavorite\")", Index=@n3) -> (ReturnValue:Text)
    InActionInfo_SortPriority=999
    InActionInfo_bIsAvailable=True
    InActionInfo_bIsEnabled=True
    InActionInfo_bIsSelected=False
    InActionInfo_bAddDivider=False
    InActionInfo_bPositionDividerAfter=False
    InActionInfo_bCloseMenuOnSelect=False
    InActionInfo_InteractionNameID=Get Action to Remove_Favorite() -> (Action to Remove_Favorite=None)
    InActionInfo_Interaction=(DataTable=None,RowName="")
    InActionInfo_PressedDelegate="(null).None"
    InActionInfo_ReleasedDelegate="(null).None"
    InInteractionDelegate=Create Event()
   )
   Return Node(execute=@n1.else, Visibility=Collapsed)
  ]}
 GetNestedButtonGroupSelectionVM {nodes [
   GetNestedButtonGroupSelectionVM() {is=title K2Node_FunctionEntry}
   > Return Node!(
    execute
    ReturnValue:Fort Nested Button Group Selection VM Object Reference=Get FortNestedButtonGroupSelectionVM() -> (FortNestedButtonGroupSelectionVM:Fort Nested Button Group Selection VM Object Reference)
   )
  ]}
 SetAnimValue {nodes [
   n1 = SetAnimValue!() -> (then, NewParam:double=~) {is=title K2Node_FunctionEntry}
   > SetScalarParameterValue(self=Get BackgroundDynamicMaterial(self=Get Block_Background()), ParameterName=Scale, Value=@n1)
  ]}
 SetBackgroundTexture {nodes [
   n1 = SetBackgroundTexture!() -> (then, Texture:Texture Object Reference) {is=title K2Node_FunctionEntry}
   > SetTextureParameterValue(self=Get BackgroundDynamicMaterial(self=Get Block_Background()), ParameterName=Thumbnail, Value=@n1)
   n2 = Get Anim_ThumbnailLoaded() -> (Anim_ThumbnailLoaded:Widget Animation Object Reference)
   > StopAnimation(InAnimation=@n2)
   > PlayAnimation(InAnimation=@n2)
  ]}
 SetCompactUIStyle {nodes [
   SetCompactUIStyle() {is=title K2Node_FunctionEntry}
   > Branch(Condition=IsMobileGame())
   > SetPadding!(
    execute
    self:Grid Slot Object Reference=SlotAsGridSlot(Widget=Get Text_Title() -> (Text_Title:Common Text Object Reference))
    InPadding_Left:float=20
    InPadding_Top:float=0
    InPadding_Right:float=152
    InPadding_Bottom:float=0
   ) -> (then)
   > Set Tile Size(self=Get WBP_Discover_Tile_Base(), InSize=(X=480,Y=176))
  ]}
 SetLoadedValue {nodes [
   n1 = SetLoadedValue!() -> (then, NewParam:double=~) {is=title K2Node_FunctionEntry}
   > SetScalarParameterValue(self=Get BackgroundDynamicMaterial(self=Get Block_Background()), ParameterName=IsThumbnailLoaded, Value=@n1)
  ]}
]
variables [
 Action to Remove_Category {type=string, default=ViewCategory}
 Action to Remove_Favorite {type=string, default=ToggleFavorite}
 AnimValue {type=number}
 BackgroundMaterialInstance {type=MaterialInstance, default=/DiscoveryBrowser/Tiles/Categories/MaterialInstances/MI_UI_Category_Background}
 Favorited Sound {type=SoundBase, default=/Game/Sounds/Fort_UI_Sounds/FrontEnd/MultiProduct/Common/Buttons/MSS_UI_MP_Button_Click_Favorite}
 LoadedValue {type=number}
 Unfavorited Sound {type=SoundBase, default=/Game/Sounds/Fort_UI_Sounds/FrontEnd/MultiProduct/Common/Buttons/MSS_UI_MP_Button_Click_Unfavorite}
]
widgets {
 info {
  inheritedWidgetCount=0
  namedSlotCount=2
  parentClass=/DiscoveryBrowser/Tiles/WBP_Discover_Tile_Button.WBP_Discover_Tile_Button_C
  rootWidgetClass=null
  widgetCount=6
 }
 widgets [
  GridPanel_52 {widgetClassPath=&GridPanel, widget=GridPanel_52, properties {columnFill [1], rowFill [1]}}
  WBP_CaptureForPostBufferUpdate_82 {
   widgetClassPath=&WBP_CaptureForPostBufferUpdate
   bIsVariable=true
   parent=GridPanel_52
   widget=WBP_CaptureForPostBufferUpdate_82
   slot {class=GridSlot, properties {layer=-1}}
  }
  Block_Background {
   widgetClassPath=/FortUILibrary/Systems/ModularButton/Blocks/WBP_UIKit_Block_BackgroundGeneric
   bIsVariable=true
   parent=GridPanel_52
   widget=Block_Background
   properties {backgroundMaterial=&MI_UI_Category_Background}
   slot {class=GridSlot}
  }
  Text_Title {
   widgetClassPath=&CommonTextBlock
   bIsVariable=true
   parent=GridPanel_52
   widget=Text_Title
   properties {
    autoWrapText=true
    mobileFontSizeMultiplier=1.5
    text=TEXT FIELD
    font {
     bForceMonospaced=false
     bMaterialIsStencil=false
     fontMaterial=null
     fontObject=&HeadingNow
     letterSpacing=10
     monospacedWidth=1
     size=15.75
     skewAmount=0
     typefaceFontName=86 Bold
     outlineSettings {
      bApplyOutlineToDropShadows=false
      bMiteredCorners=false
      bSeparateFillAlpha=false
      outlineMaterial=/FortUILibrary/MaterialInstances/Text/MI_UI_Block_Text_Outline
      outlineSize=0
      outlineColor=<rgba 0 0 0 1>
     }
    }
    shadowOffset=<0 0>
   }
   slot {class=GridSlot, properties {layer=1, verticalAlignment=VAlign_Center, padding=<ltrb 24 0 24 0>}}
  }
  GridPanel_40 {widgetClassPath=&GridPanel, widget=GridPanel_40, properties {columnFill [1], rowFill [1]}}
  StatusIndicator_Favorite {
   widgetClassPath=/FortUILibrary/UIKit/StatusIndicator/WBP_UIKit_StatusIndicator
   bIsVariable=true
   parent=GridPanel_40
   widget=StatusIndicator_Favorite
   properties {
    type=Utility
    icon=/DiscoveryBrowser/Tiles/Islands/Materials/MI_UIKit_Icon_Favorite_Discover
    iconSize=Compact
    overrideStatusIndicatorVisibility=true
    visibility=Collapsed
   }
   slot {class=GridSlot, properties {horizontalAlignment=HAlign_Left, verticalAlignment=VAlign_Bottom, padding=<ltrb 8 0 0 8>}}
  }
 ]
}
